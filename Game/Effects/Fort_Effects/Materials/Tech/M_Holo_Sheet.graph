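# Unlit additive hologram sheet material for mesh and Niagara mesh particles, previewed on the bot turret's scanner ray mesh.
# An inverted DepthFade creates a bright intersection highlight that boosts particle-colored emissive through MF_FX_EmissiveAdjustments.
# Opacity combines an inverted-V gradient with a sine-based U mask that dims toward CenterOpacity between bright bands.
# Particle alpha scales it, and a second depth fade softens contact while the intersection term keeps that line visible.
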Material /Game/Effects/Fort_Effects/Materials/Tech/M_Holo_Sheet
graph {
 properties {
  bUsedWithMeshParticles=true
  bUsedWithNiagaraMeshParticles=true
  blendMode=BLEND_Additive
  pixelDepthOffsetMode=PDOM_Legacy
  previewMesh=/Game/Gadgets/Assets/Bot_Turret/Meshes/Sm_BotTurret_ScannerRay_01
  shadingModel=MSM_Unlit
  usedShadingModels=Unlit
 }
 nodes [
  n1 = ParticleColor()
  n2 = ScalarParameter() {defaultValue=12, parameterName=IntersectDepthFade}
  n3 = Power(Base=OneMinus(None=DepthFade(FadeDistance=@n2)), Exp=ScalarParameter() {defaultValue=2.5, parameterName=IntersectPower})
  n4 = Multiply(A=ScalarParameter() {defaultValue=8, parameterName=IntersectBrightness}, B=@n3) {constB=5}
  MaterialFunctionCall(Emissive=Multiply(A=@n1.RGB, B=Add(A=@n4))) {materialFunction=&MF_FX_EmissiveAdjustments}
  n5 = TextureCoordinate()
  n6 = OneMinus(None=@n5)
  n7 = Sine(None=ComponentMask(None=@n5) {r=true})
  n8 = OneMinus(None=Saturate(None=Abs(None=@n7)))
  n9 = Multiply(
   A=Power(Base=ComponentMask(None=@n6) {g=true})
   B=LinearInterpolate(A=ScalarParameter() {defaultValue=0.3, parameterName=CenterOpacity}, Alpha=Power(Base=@n8) {constExponent=3})
  )
  n10 = Multiply(A=@n3) {constB=0.1}
  n11 = ScalarParameter() {defaultValue=40, parameterName=DepthFadeDIst}
  Multiply(A=Multiply(A=@n1.A, B=Add(A=@n9, B=@n10)), B=Add(A=@n3, B=DepthFade(FadeDistance=@n11)))
 ]
}
