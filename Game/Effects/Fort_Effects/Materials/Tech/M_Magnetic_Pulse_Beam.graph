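# Unlit, additive, two-sided Niagara mesh and ribbon material for a magnetic pulse beam effect.
# Tiled rings and pulses scroll along V by dynamic TimeAdd, each built from offset two-color sphere masks for chromatic fringing.
# An axis-aligned fresnel adds the inner beam, while particle alpha, V falloff, edge fresnel and depth fade shape opacity.
# The pulse mask also drives normal-direction world position offset, sine-tapered along the beam and scaled by WPOMult.
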Material /Game/Effects/Fort_Effects/Materials/Tech/M_Magnetic_Pulse_Beam
graph {
 properties {
  bUsedWithNiagaraMeshParticles=true
  bUsedWithNiagaraRibbons=true
  blendMode=BLEND_Additive
  pixelDepthOffsetMode=PDOM_Legacy
  shadingModel=MSM_Unlit
  twoSided=true
  usedShadingModels=Unlit
 }
 nodes [
  Constant() {r=1}
  n1 = ParticleColor()
  n2 = ComponentMask(None=TextureCoordinate()) {g=true}
  n3 = LinearInterpolate(A=DynamicParameter() {paramNames [WPOMult,TimeAdd,InnerBeamExp,BeamTile]}.BeamTile, Alpha=@n2) {constB=0}
  n4 = Multiply(A=@n2) {constB=3}
  n5 = Multiply(A=@n1.A, B=Saturate(None=Multiply(A=@n3, B=@n4)))
  n6 = ScalarParameter() {defaultValue=0.2, parameterName=ChromaticOffset}
  n7 = ScalarParameter() {defaultValue=0.3, parameterName=RingRadius}
  n8 = Constant() {r=0.5}
  n9 = ScalarParameter() {defaultValue=10, parameterName=RingTile}
  n10 = DynamicParameter() {paramNames [WPOMult,TimeAdd,InnerBeamExp,BeamTile]}
  n11 = ComponentMask(None=Multiply(A=TextureCoordinate(), B=@n10.BeamTile)) {g=true}
  n12 = ScalarParameter() {defaultValue=0.5, parameterName=RingSpeed}
  n13 = Add(A=@n10.TimeAdd) {constB=0}
  n14 = Fmod(A=Add(A=Multiply(A=@n9, B=@n11), B=Multiply(A=@n12, B=@n13)), B=Constant() {r=1})
  n15 = SphereMask(A=@n14, B=Add(A=Multiply(A=@n6, B=@n7), B=@n8), Radius=@n7) {hardnessPercent=0}
  n16 = VectorParameter() {parameterName=Color1, defaultValue=<rgba 1 0.0521 0.4094 0>}
  n17 = SphereMask(A=@n14, B=@n8, Radius=@n7) {hardnessPercent=0}
  n18 = VectorParameter() {parameterName=Color2, defaultValue=<rgba 0 1.25 1.1248 0>}
  n19 = ScalarParameter() {defaultValue=0.2, parameterName=ChromaticOffset}
  n20 = ScalarParameter() {defaultValue=0.2, parameterName=PulseRadius}
  n21 = Constant() {r=0.5}
  n22 = ScalarParameter() {defaultValue=1, parameterName=PulseTile}
  n23 = ScalarParameter() {defaultValue=1, parameterName=PulseSpeed}
  n24 = Fmod(A=Add(A=Multiply(A=@n11, B=@n22), B=Multiply(A=@n13, B=@n23)), B=Constant() {r=1})
  n25 = SphereMask(A=@n24, B=Add(A=Multiply(A=@n19, B=@n20), B=@n21), Radius=@n20) {hardnessPercent=0}
  n26 = SphereMask(A=@n24, B=@n21, Radius=@n20) {hardnessPercent=0}
  n27 = Add(
   A=LinearInterpolate(A=Multiply(A=@n15, B=@n16.RGB), B=Multiply(A=@n17, B=@n18.RGB))
   B=LinearInterpolate(A=Multiply(A=@n16.RGB, B=@n25), B=Multiply(A=@n18.RGB, B=@n26))
  )
  n28 = ScalarParameter() {defaultValue=2, parameterName=Power}
  n29 = MaterialFunctionCall() {materialFunction=&TangentBasis}
  n30 = Multiply(
   A=DynamicParameter() {paramNames [WPOMult,TimeAdd,InnerBeamExp,BeamTile]}.InnerBeamExp
   B=ScalarParameter() {defaultValue=40, parameterName=FresnelExp}
  ) {constB=0.4}
  n31 = Add(
   A=Desaturation(None=Power(Base=@n27, Exp=@n28), Fraction=Constant() {r=-1})
   B=Multiply(
    A=VectorParameter() {parameterName=BeamColor, defaultValue=<rgba 0.2023 0.4581 1 0>}.RGB
    B=MaterialFunctionCall(Axis=@n29.Y V, Exponent=@n30) {materialFunction=&AxisAlignedFresnel}.Fresnel
   ) {constB=0.5}
  ) {constB=0}
  n32 = ScalarParameter() {defaultValue=2, parameterName=EdgeFadeExp}
  DepthFade(
   Opacity=Multiply(
    A=Multiply(A=@n5, B=@n31)
    B=MaterialFunctionCall(Axis=@n29.Y V, Exponent=@n32) {materialFunction=&AxisAlignedFresnel}.Fresnel
   )
   FadeDistance=ScalarParameter() {defaultValue=10, parameterName=DepthFade}
  )
  n33 = Saturate(None=TwoSidedSign())
  MaterialFunctionCall(Emissive=Multiply(A=LinearInterpolate(Alpha=@n33) {constA=0.5}, B=Multiply(A=@n1.RGB, B=@n31))) {materialFunction=&MF_FX_EmissiveAdjustments}
  n34 = Sine(None=ComponentMask(None=TextureCoordinate() {vTiling=0.5}) {g=true})
  n35 = Normalize(VectorInput=VertexNormalWS())
  Multiply(
   A=Multiply(A=@n26, B=Multiply(A=@n34, B=@n35))
   B=Multiply(
    A=ScalarParameter() {defaultValue=3, parameterName=WPOPulseSize}
    B=DynamicParameter() {paramNames [WPOMult,TimeAdd,InnerBeamExp,BeamTile]}.WPOMult
   )
  )
  Time()
 ]
}
